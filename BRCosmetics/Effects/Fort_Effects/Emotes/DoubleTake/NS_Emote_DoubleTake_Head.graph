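# Niagara system for the head effect of the DoubleTake emote, emitting short-lived diamond sprites around points sampled on a user-supplied skeletal mesh.
# Diamonds spawn continuously in a scaled sphere shape, receive velocity relative to the sampled point, then slow under drag.
# Random cyan-tinted colors dim and fade over life as size pops up, then shrinks.
# A disabled Light emitter holds two flickering cyan lights, offset oppositely by spawn group.
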NiagaraSystem /BRCosmetics/Effects/Fort_Effects/Emotes/DoubleTake/NS_Emote_DoubleTake_Head
dependencies {
 usedDataInterfaces [&NiagaraDataInterfaceCurve, &NiagaraDataInterfaceColorCurve]
 usedDynamicInputs [/Niagara/DynamicInputs/Helpers/MaskBoolBySpawnGroup, &SetBoolByFloatComparison]
 usedModules [&SystemState]
}
emitters {
 Diamonds {
  emitterUpdateScript [
   &EmitterState()
   &SpawnRate(SpawnRate=10, Use Spawn Probability=false)
  ]
  particleSpawnScript [
   &InitializeParticle(
    Color Maximum=<rgba 0.0417 1.0586 2 1>
    Color Minimum=<rgba 0.4062 0.7427 1 1>
    Color Mode=::NewEnumerator2
    Color Randomness Mode:ENiagaraRandomnessMode=::NewEnumerator2
    Lifetime Max=0.5
    Lifetime Min=0.35
    Lifetime Mode=::NewEnumerator1
    Sprite Rotation Mode=::NewEnumerator0
    Sprite Size Mode=::NewEnumerator1
    Uniform Sprite Size Max=13
    Uniform Sprite Size Min=8
    UsePositionOffset=false
    Write Material Random=true
   )
   &SkeletalMeshLocation(Bone Sampling Mode=::NewEnumerator4, MaskSpawnGroup=false, Skeletal Mesh=$User.SkeletalMesh)
   &ShapeLocation(
    Hemisphere Distribution=<1 0.6>
    MaskSpawnGroup=false
    Non Uniform Scale=<1.5 1.5 2>
    Override Seed=false
    Sphere Radius=10
    Sphere Surface Distribution=1
    UseOwnerScale=true
   )
   &AddVelocity(
    Module.Use Falloff Distance=false
    Override Seed=false
    Velocity Mode=::NewEnumerator1
    Velocity Origin=$Output.SkeletalMeshLocation.SampledPositionWithoutOffset
   )
   &CameraOffset(Camera Offset Amount=&RandomRangeFloat(Minimum=0, Maximum=5))
  ]
  particleUpdateScript [
   &ParticleState()
   &Drag(Drag=12, Use Linear Drag=true, Use Rotational Drag=true)
   &SolveForcesAndVelocity(Clamp Velocity=false, Limit Acceleration=false)
   &ScaleColor(
    Scale Alpha=&FloatFromCurve(FloatCurve=DataInterface() {propertyValues {Curve {keys [
         {tangentMode=RCTM_User, time=0, value=1}
         {arriveTangent=-0.0405, interpMode=RCIM_Cubic, leaveTangent=-0.0405, tangentMode=RCTM_User, time=0.7009, value=1}
         {arriveTangent=-8.0089, interpMode=RCIM_Cubic, leaveTangent=-8.0089, tangentMode=RCTM_User, time=1, value=0}
        ]}}})
    Scale RGB=&MakeVectorFromLinearColorRGB(Color=&ColorFromCurve(ColorCurve=DataInterface() {propertyValues {
        AlphaCurve {keys []}
        BlueCurve {keys [<at 0 1.5>,<at 0.888 0.5>,<at 0.9706 0>]}
        GreenCurve {keys [<at 0 1.5>,<at 0.888 0.5>,<at 0.9706 0>]}
        RedCurve {keys [<at 0 1.5>,<at 0.888 0.5>,<at 0.9706 0>]}
       }}))
    ScaleA=true
    ScaleRGB=true
   )
   &ScaleSpriteSize(Uniform Curve Sprite Scale=DataInterface() {propertyValues {Curve {keys [
        {arriveTangent=9.0725, interpMode=RCIM_Cubic, leaveTangent=9.0725, tangentMode=RCTM_User, time=0, value=0}
        {arriveTangent=-2.5249, interpMode=RCIM_Cubic, leaveTangent=-2.5249, tangentMode=RCTM_User, time=0.1408, value=0.9847}
        {arriveTangent=-0.0405, interpMode=RCIM_Cubic, leaveTangent=-0.0405, tangentMode=RCTM_User, time=1, value=0}
       ]}}})
  ]
  rendererClasses [&NiagaraSpriteRendererProperties]
  renderers [{rendererClass=&NiagaraSpriteRendererProperties}]
 }
 Light {
  bEnabled=false
  emitterUpdateScript [
   &EmitterState()
   &SpawnBurst_Instantaneous(Use Spawn Probability=false, UseLoopCountLimit=true)
   &SpawnBurst_Instantaneous(Spawn Group=1, Use Spawn Probability=false, UseLoopCountLimit=true) {module=SpawnBurst_Instantaneous001}
  ]
  particleSpawnScript [&InitializeParticle(Color Mode=::NewEnumerator0, Lifetime=0.25, UsePositionOffset=false, Write Material Random=true)]
  particleUpdateScript [
   &ParticleState(
    Kill Particles When Lifetime Has Elapsed=false
    Let Infinitely Lived Particles Die When Emitter Deactivates=true
    Loop Particles Lifetime=true
   )
   &SkeletalMeshLocation(
    Bone / Socket Index=&SpawnGroup()
    Bone Sampling Mode=::NewEnumerator4
    MaskSpawnGroup=false
    Sampled Position Offset=&MaskVectorBySpawnGroup(Masked Vector=<7.5 0 0>, Passthrough Vector=<-7.5 0 0>)
    Skeletal Mesh=$User.SkeletalMesh
   )
   &FN_Light_Attributes(
    Color=<rgba 0.1624 0.9903 2 1>
    Create Color Attribute=true
    Create Exponent Attribute=true
    Create Radius Attribute=true
    Exponent=2
    Radius=100
    Scale Color=&FloatFromCurve(
     FloatCurve=DataInterface() {propertyValues {Curve {keys [
         <at 0 1 RCIM_Cubic>
         {arriveTangent=1.657, interpMode=RCIM_Cubic, leaveTangent=1.657, time=0.052, value=0.8043}
         {arriveTangent=-0.0227, interpMode=RCIM_Cubic, leaveTangent=-0.0227, tangentMode=RCTM_User, time=0.1586, value=1.2628}
         {arriveTangent=-2.6232, interpMode=RCIM_Cubic, leaveTangent=-2.6232, time=0.1971, value=0.3384}
         {arriveTangent=2.772, interpMode=RCIM_Cubic, leaveTangent=2.772, time=0.2801, value=0.9442}
         {arriveTangent=1.7032, interpMode=RCIM_Cubic, leaveTangent=1.7032, time=0.3617, value=0.7947}
         {arriveTangent=-2.2884, interpMode=RCIM_Cubic, leaveTangent=-2.2884, time=0.4312, value=1.2016}
         {arriveTangent=-2.9025, interpMode=RCIM_Cubic, leaveTangent=-2.9025, time=0.4644, value=0.5596}
         {arriveTangent=-1.6052, interpMode=RCIM_Cubic, leaveTangent=-1.6052, time=0.5221, value=0.9378}
         {arriveTangent=0.5696, interpMode=RCIM_Cubic, leaveTangent=0.5696, time=0.5745, value=0.3829}
         {arriveTangent=2.6411, interpMode=RCIM_Cubic, leaveTangent=2.6411, time=0.6635, value=1.0184}
         {arriveTangent=2.9949, interpMode=RCIM_Cubic, leaveTangent=2.9949, time=0.7532, value=0.855}
         {arriveTangent=-1.4133, interpMode=RCIM_Cubic, leaveTangent=-1.4133, time=0.8257, value=1.5042}
         {arriveTangent=-2.8923, interpMode=RCIM_Cubic, leaveTangent=-2.8923, time=0.878, value=0.6785}
         {interpMode=RCIM_Cubic, leaveTangent=0.3621, time=1, value=1}
        ]}}}
     Scale Curve=2
    )
   )
  ]
  rendererClasses [&NiagaraLightRendererProperties]
  renderers [{rendererClass=&NiagaraLightRendererProperties, rendererIndex=0}]
 }
}
summary {
 systemName=NS_Emote_DoubleTake_Head
 emitters [
  {bEnabled=false, emitterName=Light, rendererClasses [&NiagaraLightRendererProperties]}
  {emitterName=Diamonds, rendererClasses [&NiagaraSpriteRendererProperties]}
 ]
 userVariables [
  User.SkeletalMesh {
   type=NiagaraDataInterfaceSkeletalMesh
   defaultValue {
    struct=NiagaraExt_VariableValue_DataInterface
    value {dataInterface=&self:NiagaraDataInterfaceSkeletalMesh_0}
   }
  }
  User.Spawn Left {type=NiagaraFloat, defaultValue=0, description=""}
  User.Spawn Right {type=NiagaraFloat, defaultValue=0, description=""}
 ]
}
